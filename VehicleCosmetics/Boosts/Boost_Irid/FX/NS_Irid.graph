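# Niagara system for the Irid vehicle boost cosmetic, drawing trails and thruster flames at the car's skeletal mesh sockets.
# The active ribbon emitter spawns per socket, scaling rate and velocity by game mode, and fades in while inheriting source movement.
# Two looping mesh/sprite thruster emitters derive HueShift from emitter age, offset from each other, to cycle hue.
# Four other emitters are disabled, including the painted-colour multi-ribbon layers and a rainbow-curve thruster.
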NiagaraSystem /VehicleCosmetics/Boosts/Boost_Irid/FX/NS_Irid
dependencies {
 usedDataInterfaces [
  &NiagaraDataInterfaceCurve
  &NiagaraDataInterfaceColorCurve
  &NiagaraDataInterfaceVectorCurve
  &NiagaraDataInterfaceSkeletalMesh
  &NiagaraDataInterfaceArrayFloat3
 ]
 usedDynamicInputs [&SelectVectorFromArray]
 usedModules [
  &self:SystemSpawnScript.SystemScriptSource.SystemScriptGraph.NiagaraNodeAssignment_0.SetVariables_#1
  &self:SystemSpawnScript.SystemScriptSource.SystemScriptGraph.NiagaraNodeAssignment_1.SetVariables_#2
  &self:SystemSpawnScript.SystemScriptSource.SystemScriptGraph.NiagaraNodeAssignment_2.SetVariables_#3
  &SystemState
 ]
}
emitters {
 EnergyStrike_Optimized {
  bEnabled=false
  emitterUpdateScript [
   &EmitterState()
   &VC_FilteredSocketCount()
   &VC_RateQualityHandler(Default Rate:NiagaraFloat=&Multiply_Float(
     A=100
     B=/VehicleCosmetics/FX/Niagara/Modules/VC_Mask_Floats(Coro Value:NiagaraFloat=1, FrontEnd Value:NiagaraFloat=0.75, FN Value:NiagaraFloat=0.5)
    ))
   &VC_SpawnRibbon_v1(Number of Ribbons=$Emitter.PortCount, SpawnRate=80)
  ]
  particleSpawnScript [
   &VC_PaintedVariantsOverride() {enabled=false}
   &VC_InitializeRibbon_v1(
    Custom Offset=true
    Enable Skeletal Position=true
    Lifetime=0.7
    Local Vector Offset=&AddVectorToPosition(Position=<10 0 0>, Vector=$User.DTCarOffset001)
    Offset Amount=1
    Ribbon Width:NiagaraFloat=&MakeCustomFloatFromBool(Bool=$User.bIsFrontendPreview, True Float=30, False Float=40)
   )
   &AddVelocity(Velocity=<-1500 0 0>) {module=AddVelocity001}
   &Color(Color=&MakeCustomLinearColorFromBool(Bool=$User.bIsPainted, True Color=$User.PaintedColor, False Color=<rgba 0.102 1 0.102 0.502>))
   /SetVariables(
    Particles.WidthRibbon01:NiagaraFloat=&MakeCustomFloatFromBool(Bool=$User.bIsFrontendPreview, True Float=15, False Float=25)
    Particles.WidthRibbon02:NiagaraFloat=&MakeCustomFloatFromBool(Bool=$User.bIsFrontendPreview, True Float=25, False Float=35)
    Particles.WidthRibbon03:NiagaraFloat=&MakeCustomFloatFromBool(Bool=$User.bIsFrontendPreview, True Float=30, False Float=40)
   ) {isSetParameters=true}
   /VehicleCosmetics/FX/Niagara/Modules/MergeEmitters/VC_ExtraBindingColors(
    ExtraColors=3
    R1 Color:LinearColor=&MakeCustomLinearColorFromBool(Bool=$User.bIsPainted, True Color=$User.PaintedColor, False Color=<rgba 0.6 0.451 0.149 20>)
    R2 Color:LinearColor=&MakeCustomLinearColorFromBool(Bool=$User.bIsPainted, True Color=$User.PaintedColor, False Color=<rgba 0.6 1 0.2 0.502>)
    R3 Color:LinearColor=&MakeCustomLinearColorFromBool(Bool=$User.bIsPainted, True Color=$User.PaintedColor, False Color=<rgba 0.349 1 0.4 0.502>)
   )
   /VehicleCosmetics/FX/VC_FXParents/Modules/DLR_ProductPreviews(GameModeID:NiagaraInt32=$System.GameModeID, bPreviewVehicleSpeed:NiagaraBool=$System.bPreviewVehicleSpeed)
  ]
  particleUpdateScript [
   &ParticleState()
   &ScaleVelocity(Velocity Scale=&VectorFromFloat(Value=&FloatFromCurve(FloatCurve=DataInterface() {propertyValues {Curve {keys [<at 0 0>,<at 1 2>]}}}))) {enabled=false}
   &InheritSourceMovement(
    Applied Position Scale=&Multiply_VectorByFloat(
     Vector=<1 1 1>
     Float=&FloatFromCurve(FloatCurve=DataInterface() {propertyValues {Curve {keys [<at 0 1 RCIM_Cubic>, {interpMode=RCIM_Cubic, leaveTangent=-3.4326, time=0.7, value=0}]}}})
    )
    ApplyForce=false
    ApplyPosition=true
    ApplyPositionFalloff=false
   ) {module=InheritSourceMovement001}
   &InheritSourceMovement(
    Applied Position Scale=&VectorFromFloat(Value=&FloatFromCurve(FloatCurve=DataInterface() {propertyValues {Curve {
         defaultValue=1
         keys [
          {arriveTangent=-1.7885, interpMode=RCIM_Cubic, leaveTangent=-1.7885, tangentMode=RCTM_User, time=0.05, value=1}
          {interpMode=RCIM_Cubic, tangentMode=RCTM_User, time=1, value=0}
         ]
        }}}))
    ApplyForce=false
    ApplyPosition=true
    ApplyPositionFalloff=false
   )
   &SolveForcesAndVelocity(Clamp Velocity=false, Limit Acceleration=false)
   /SetVariables(
    Particles.R1 Color:LinearColor=&VC_DynamicScaleColor(
     Color to Scale:LinearColor=&MakeCustomLinearColorFromBool(Bool=$User.bIsPainted, True Color=$User.PaintedColor, False Color=<rgba 0.6 0.451 0.149 0.5>)
     RGB Scale:Vector3f=&VectorFromFloat(Value=&FloatFromCurve(
       FloatCurve=DataInterface() {propertyValues {Curve {keys [<at 0 0.35>,<at 1 0>]}}}
       Scale Curve=&MakeCustomFloatFromBool(Bool=$User.bIsPainted, True Float=$User.Power_EmissivePaintedR2, False Float=1)
      ))
     Alpha Scale:NiagaraFloat=&FloatFromCurve(
      FloatCurve=DataInterface() {propertyValues {Curve {keys [
          <at 0 1 RCIM_Cubic>
          {arriveTangent=-0.2383, interpMode=RCIM_Cubic, leaveTangent=-0.2383, tangentMode=RCTM_User, time=0.35, value=1}
          <at 1 0 RCIM_Cubic>
         ]}}}
      Scale Curve=&MakeCustomFloatFromBool(Bool=$User.bIsPainted, True Float=$User.Power_AlphaPaintedR2, False Float=1)
     )
    )
    Particles.R2 Color:LinearColor=&VC_DynamicScaleColor(
     Color to Scale:LinearColor=&MakeCustomLinearColorFromBool(Bool=$User.bIsPainted, True Color=$User.PaintedColor, False Color=<rgba 0.6 1 0.2 0.502>)
     RGB Scale:Vector3f=&VectorFromFloat(Value=&FloatFromCurve(
       FloatCurve=DataInterface() {propertyValues {Curve {keys [<at 0 2 RCIM_Cubic>,<at 1 0 RCIM_Cubic>]}}}
       Scale Curve=&MakeCustomFloatFromBool(Bool=$User.bIsPainted, True Float=$User.Power_EmissivePaintedR3, False Float=2)
      ))
     Alpha Scale:NiagaraFloat=&FloatFromCurve(
      FloatCurve=DataInterface() {propertyValues {Curve {keys [
          <at 0 5 RCIM_Cubic>
          {arriveTangent=-0.2383, interpMode=RCIM_Cubic, leaveTangent=-0.2383, tangentMode=RCTM_User, time=0.8, value=1}
          <at 1 0 RCIM_Cubic>
         ]}}}
      Scale Curve=&MakeCustomFloatFromBool(Bool=$User.bIsPainted, True Float=$User.Power_AlphaPaintedR3, False Float=1)
     )
    )
    Particles.R3 Color:LinearColor=&VC_DynamicScaleColor(
     Color to Scale:LinearColor=&MakeCustomLinearColorFromBool(Bool=$User.bIsPainted, True Color=$User.PaintedColor, False Color=<rgba 0.349 1 0.4 0.502>)
     RGB Scale:Vector3f=&VectorFromFloat(Value=&FloatFromCurve(
       FloatCurve=DataInterface() {propertyValues {Curve {keys [<at 0 8 RCIM_Cubic>,<at 1 1 RCIM_Cubic>]}}}
       Scale Curve=&MakeCustomFloatFromBool(Bool=$User.bIsPainted, True Float=$User.Power_EmissivePaintedR4, False Float=1)
      ))
     Alpha Scale:NiagaraFloat=&FloatFromCurve(
      FloatCurve=DataInterface() {propertyValues {Curve {keys [
          <at 0 8 RCIM_Cubic>
          {arriveTangent=-0.2383, interpMode=RCIM_Cubic, leaveTangent=-0.2383, tangentMode=RCTM_User, time=0.5, value=8}
          <at 1 0 RCIM_Cubic>
         ]}}}
      Scale Curve=&MakeCustomFloatFromBool(Bool=$User.bIsPainted, True Float=$User.Power_AlphaPaintedR4, False Float=1.5)
     )
    )
   ) {isSetParameters=true}
   &ScaleColor(
    Scale Alpha=&FloatFromCurve(
     FloatCurve=DataInterface() {propertyValues {Curve {keys [
         <at 0 2 RCIM_Cubic>
         {arriveTangent=-0.2383, interpMode=RCIM_Cubic, leaveTangent=-0.2383, tangentMode=RCTM_User, time=0.8, value=1}
         <at 1 0 RCIM_Cubic>
        ]}}}
     Scale Curve=&MakeCustomFloatFromBool(Bool=$User.bIsPainted, True Float=$User.Power_AlphaPaintedR1, False Float=1)
    )
    Scale RGB=&VectorFromFloat(Value=&FloatFromCurve(
      FloatCurve=DataInterface() {propertyValues {Curve {keys [<at 0 1>,<at 1 0>]}}}
      Scale Curve=&MakeCustomFloatFromBool(Bool=$User.bIsPainted, True Float=$User.Power_EmissivePaintedR1, False Float=1)
     ))
    ScaleA=true
    ScaleRGB=true
   )
   &DynamicMaterialParameters(
    Index 0 Param 1=5
    Index 0 Param 2=0.5
    Index 0 Param 4=1
    Index 1 Param 2=&MakeCustomFloatFromBool(Bool=$User.bIsPainted, True Float=$User.HueShift_M1, False Float=0.05)
    Module.Index0Param0Write=true
    Module.Index0Param1Write=true
    Module.Index0Param2Write=true
    Module.Index0Param3Write=true
    Module.Index1Param0Write=true
    Module.Index1Param1Write=true
    Module.Index1Param2Write=false
    Module.Index1Param3Write=false
    Write Parameter Index 1=true
   )
   &ScaleRibbonWidth(Ribbon Width Scale=&FloatFromCurve(FloatCurve=DataInterface() {propertyValues {Curve {keys [{interpMode=RCIM_Cubic, tangentMode=RCTM_Break, time=0, value=3}, {interpMode=RCIM_Cubic, tangentMode=RCTM_Break, time=1, value=0.5}]}}}))
   /SetVariables(
    Particles.Index_Dynamic1_01:Vector4f=&MakeVector4(X=1, Y=0.1, Z=1, W=10)
    Particles.Index_Dynamic2_01:Vector4f=&MakeVector4(
     X=0
     Y=&MakeCustomFloatFromBool(Bool=$User.bIsPainted, True Float=$User.HueShift_M4, False Float=0.2)
     Z=0
     W=0
    )
   ) {isSetParameters=true}
   /SetVariables(
    Particles.Index_Dynamic1_02:Vector4f=&MakeVector4(X=1, Y=1, Z=1, W=3)
    Particles.Index_Dynamic2_02:Vector4f=&MakeVector4(
     X=0
     Y=&MakeCustomFloatFromBool(Bool=$User.bIsPainted, True Float=$User.HueShift_M2, False Float=0)
     Z=0
     W=0
    )
   ) {isSetParameters=true}
   /SetVariables(
    Particles.Index_Dynamic1_03:Vector4f=&MakeVector4(X=0.35, Y=1, Z=0.05, W=5)
    Particles.Index_Dynamic2_03:Vector4f=&MakeVector4(
     X=0
     Y=&MakeCustomFloatFromBool(Bool=$User.bIsPainted, True Float=$User.HueShift_M3, False Float=0.06)
     Z=0
     W=0
    )
   ) {isSetParameters=true}
   /SetVariables(
    Particles.ScaleWidthR01:NiagaraFloat=&Multiply_Float(
     A=$Particles.WidthRibbon01
     B=&FloatFromCurve(FloatCurve=DataInterface() {propertyValues {Curve {keys [{interpMode=RCIM_Cubic, tangentMode=RCTM_Break, time=0, value=3}, {interpMode=RCIM_Cubic, tangentMode=RCTM_Break, time=1, value=0.5}]}}})
    )
    Particles.ScaleWidthR02:NiagaraFloat=&Multiply_Float(
     A=$Particles.WidthRibbon02
     B=&FloatFromCurve(FloatCurve=DataInterface() {propertyValues {Curve {keys [{interpMode=RCIM_Cubic, tangentMode=RCTM_Break, time=0, value=3}, {interpMode=RCIM_Cubic, tangentMode=RCTM_Break, time=1, value=0.5}]}}})
    )
    Particles.ScaleWidthR03:NiagaraFloat=&Multiply_Float(
     A=$Particles.WidthRibbon03
     B=&FloatFromCurve(FloatCurve=DataInterface() {propertyValues {Curve {keys [{interpMode=RCIM_Cubic, tangentMode=RCTM_Break, time=0, value=3}, {interpMode=RCIM_Cubic, tangentMode=RCTM_Break, time=1, value=0.5}]}}})
    )
   ) {isSetParameters=true}
   &VC_UpdateRibbon_v1(Enable Skeletal Position=true)
  ]
  rendererClasses [&NiagaraRibbonRendererProperties]
  renderers [
   {rendererClass=&NiagaraRibbonRendererProperties}
   {rendererClass=&NiagaraRibbonRendererProperties, rendererIndex=1}
   {rendererClass=&NiagaraRibbonRendererProperties, rendererIndex=2}
   {rendererClass=&NiagaraRibbonRendererProperties, rendererIndex=3}
  ]
 }
 NE_VC_Boost_Ribbon_Parent {
  emitterUpdateScript [
   &EmitterState()
   &VC_FilteredSocketCount()
   &VC_RateQualityHandler(Default Rate:NiagaraFloat=&Multiply_Float(
     A=100
     B=&VC_Mask_Floats(Coro Value:NiagaraFloat=1, FrontEnd Value:NiagaraFloat=0.75, FN Value:NiagaraFloat=0.75)
    ))
   &VC_SpawnRibbon_v1(Number of Ribbons=$Emitter.PortCount, SpawnRate=$Emitter.Spawn Rate)
  ]
  particleSpawnScript [
   &VC_PaintedVariantsOverride() {enabled=false}
   &VC_InitializeRibbon_v1(Color=<rgba 1 1 1 3>, Enable Skeletal Position=true, Lifetime=1, Ribbon Width=100)
   &AddVelocity(Velocity=&VC_Mask_Vectors(Coro Value:Vector3f=<-400 0 0>, FrontEnd Value:Vector3f=<-1200 0 0>, FN Value:Vector3f=<-400 0 0>)) {module=AddVelocity001}
   &DLR_ProductPreviews(GameModeID:NiagaraInt32=$System.GameModeID, bPreviewVehicleSpeed:NiagaraBool=$System.bPreviewVehicleSpeed)
  ]
  particleUpdateScript [
   &ParticleState()
   &ScaleColor(
    Scale Alpha=&FloatFromCurve(FloatCurve=DataInterface() {propertyValues {Curve {keys [<at 0 0>,<at 0.25 1>]}}})
    ScaleA=true
    ScaleRGB=true
   )
   &InheritSourceMovement(
    Applied Position Scale=&VectorFromFloat(Value=&FloatFromCurve(FloatCurve=DataInterface() {propertyValues {Curve {
         defaultValue=1
         keys [
          {arriveTangent=-1.7885, interpMode=RCIM_Cubic, leaveTangent=-1.7885, tangentMode=RCTM_User, time=0.05, value=1}
          {interpMode=RCIM_Cubic, tangentMode=RCTM_User, time=1, value=0}
         ]
        }}}))
    ApplyForce=false
    ApplyPosition=true
    ApplyPositionFalloff=false
   )
   &SolveForcesAndVelocity(Clamp Velocity=false, Limit Acceleration=false)
   &VC_UpdateRibbon_v1(Enable Skeletal Position=true)
  ]
  rendererClasses [&NiagaraRibbonRendererProperties]
  renderers [
   {rendererClass=&NiagaraRibbonRendererProperties}
   {rendererClass=&NiagaraRibbonRendererProperties, rendererIndex=1}
  ]
 }
 NE_VC_Boost_Ribbon_Parent001 {
  bEnabled=false
  emitterUpdateScript [
   &EmitterState()
   &VC_FilteredSocketCount()
   &VC_RateQualityHandler()
   &VC_SpawnRibbon_v1(Number of Ribbons=$Emitter.PortCount, SpawnRate=$Emitter.Spawn Rate)
  ]
  particleSpawnScript [
   &VC_PaintedVariantsOverride() {enabled=false}
   &VC_InitializeRibbon_v1(
    Enable Skeletal Position=true
    Lifetime=&MergeHelper_FloatSwitch(0=1.5, 1=1.5, 2=1.5, 3=?, Index=$User.GameModeID, Num=::NewEnumerator1)
    Ribbon Width=50
   )
   &AddVelocity(Velocity=&MakeCustomVectorFromBool(Bool=$User.bIsPainted, True Vector=<-1500 0 1>, False Vector=<-1500 0 0>)) {module=AddVelocity001}
   &DLR_ProductPreviews(GameModeID:NiagaraInt32=$System.GameModeID, bPreviewVehicleSpeed:NiagaraBool=$System.bPreviewVehicleSpeed)
  ]
  particleUpdateScript [
   &ParticleState()
   &InheritSourceMovement(
    Applied Position Scale=&VectorFromFloat(Value=&FloatFromCurve(FloatCurve=DataInterface() {propertyValues {Curve {
         defaultValue=1
         keys [
          {arriveTangent=-1.7885, interpMode=RCIM_Cubic, leaveTangent=-1.7885, tangentMode=RCTM_User, time=0.05, value=1}
          {interpMode=RCIM_Cubic, tangentMode=RCTM_User, time=1, value=0}
         ]
        }}}))
    ApplyForce=false
    ApplyPosition=true
    ApplyPositionFalloff=false
   )
   &SolveForcesAndVelocity(Clamp Velocity=false, Limit Acceleration=false)
   &VC_UpdateRibbon_v1(Enable Skeletal Position=true)
  ]
  rendererClasses [&NiagaraRibbonRendererProperties]
  renderers [{rendererClass=&NiagaraRibbonRendererProperties}]
 }
 NE_VC_Thruster {
  emitterUpdateScript [
   &EmitterState(
    Inactive Response=::NewEnumerator1
    Life Cycle Mode=::NewEnumerator1
    Loop Behavior=::NewEnumerator1
    Loop Duration Mode=::NewEnumerator1
    UseLoopDelay=false
   )
   &VC_FilteredSocketCount()
   &SpawnRate(SpawnRate=30, Use Spawn Probability=false) {enabled=false}
   &SpawnBurst_Instantaneous(Spawn Count=2, Use Spawn Probability=false, UseLoopCountLimit=false)
   /SetVariables(Emitter.HueShift:NiagaraFloat=&Divide_Float(A=$Emitter.Age, B=3)) {isSetParameters=true, module=SetVariables_#4}
  ]
  particleSpawnScript [
   &VC_PaintedVariantsOverride() {enabled=false}
   &InitializeParticle(
    Color Mode=::NewEnumerator0
    Lifetime=4
    Mesh Scale=<4.5 2 4>
    Mesh Scale Mode=::NewEnumerator4
    Position Mode=::NewEnumerator0
    Write Material Random=true
   )
   &PartitionParticles(Number Of Partitions=$Emitter.PortCount)
   &SkeletalMeshLocation(
    Bone / Socket Index=$Particles.PartitionParticles.Partition
    Bone / Socket Sampling Mode=::NewEnumerator5
    Interpolate Mesh Sampling=false
    MaskSpawnGroup=false
    Mesh Sampling Type=::NewEnumerator6
    Sampled Position Offset=&AddVectorToPosition(Position=<0 0 0>, Vector=&AddVectorToPosition(Position=$User.DTCarOffset, Vector=<39.3897 0 0>))
    Skeletal Mesh=$System.SkeletalMesh
   )
   &AddVelocity(Velocity=&RandomRangeVector(Minimum=<-60 0 0>, Maximum=<-90 1 1>)) {enabled=false}
   &SpriteFacingAndAlignment(Sprite Alignment=<1 0 0>, Sprite Facing=<0 1 0>, WriteAlignment=true, WriteFacing=true) {enabled=false}
   &InitialMeshOrientation(Mesh Orientation Mode=::NewEnumerator7, Use Rotation Vector=false) {enabled=false}
   &DLR_ProductPreviews(GameModeID:NiagaraInt32=$System.GameModeID, bPreviewVehicleSpeed:NiagaraBool=$System.bPreviewVehicleSpeed)
  ]
  particleUpdateScript [
   &ParticleState(Kill Particles When Lifetime Has Elapsed=false, Loop Particles Lifetime=true)
   &InheritSourceMovement(
    Applied Position Scale=&Multiply_VectorByFloat(
     Vector=<1 1 1>
     Float=&FloatFromCurve(FloatCurve=DataInterface() {propertyValues {Curve {
         defaultValue=1
         keys [
          {arriveTangent=-1.7885, interpMode=RCIM_Cubic, leaveTangent=-1.7885, tangentMode=RCTM_User, time=0.05, value=1}
          {interpMode=RCIM_Cubic, tangentMode=RCTM_User, time=1, value=0}
         ]
        }}})
    )
    ApplyForce=false
    ApplyPosition=true
    ApplyPositionFalloff=false
   ) {enabled=false, module=InheritSourceMovement001}
   &ScaleColor(
    Scale Alpha=&FloatFromCurve(FloatCurve=DataInterface() {propertyValues {Curve {keys [{arriveTangent=0.3506, time=0, value=1}, <at 6 0>]}}})
    ScaleA=true
    ScaleRGB=false
   ) {enabled=false}
   &DynamicMaterialParameters(
    Index 0 Param 2=1
    Index 1 Param 1=1
    Index 1 Param 2=1
    Index 1 Param 3=0.992
    Index 1 Param 4=&FloatFromCurve(FloatCurve=DataInterface() {propertyValues {Curve {keys [<at 0 2>,<at 1 0>]}}})
    Index 2 Param 1=0.9488
    Index 2 Param 3=&FloatFromCurve(FloatCurve=DataInterface() {propertyValues {Curve {keys [<at 0 0>,<at 1 1>]}}})
    Index 2 Param 4=0.032
    Module.Index0Param0Write=false
    Module.Index0Param1Write=true
    Module.Index0Param2Write=false
    Module.Index0Param3Write=false
    Module.Index1Param0Write=true
    Module.Index1Param1Write=true
    Module.Index1Param2Write=false
    Module.Index1Param3Write=true
    Module.Index2Param0Write=true
    Module.Index2Param1Write=true
    Module.Index2Param2Write=true
    Module.Index2Param3Write=true
    Module.Index3Param0Write=true
    Module.Index3Param1Write=true
    Module.Index3Param2Write=true
    Module.Index3Param3Write=true
    Write Parameter Index 1=true
    Write Parameter Index 2=true
    Write Parameter Index 3=true
   )
   &JitterPosition(Jitter Amount=0.3) {enabled=false}
   &ScaleColor(
    Linear Color Curve=DataInterface() {propertyValues {
      AlphaCurve {keys [<at 0 1>,<at 1 1>]}
      BlueCurve {keys [<at 0 1>,<at 1 1>]}
      GreenCurve {keys [<at 0 1>,<at 1 1>]}
      RedCurve {keys [<at 0 1>,<at 1 1>]}
     }}
    Scale Mode=2
   ) {module=ScaleColor001}
   /SetVariables(Particles.HueShift:NiagaraFloat=$Emitter.Age) {isSetParameters=true, module=SetVariables_#5}
   &SolveForcesAndVelocity(Clamp Velocity=false, Limit Acceleration=false)
  ]
  rendererClasses [&NiagaraMeshRendererProperties, &NiagaraSpriteRendererProperties]
  renderers [
   {rendererClass=&NiagaraMeshRendererProperties, rendererIndex=0}
   {rendererClass=&NiagaraSpriteRendererProperties, rendererIndex=1}
  ]
 }
 NE_VC_Thruster001 {
  bEnabled=false
  emitterUpdateScript [
   &EmitterState(
    Inactive Response=::NewEnumerator1
    Life Cycle Mode=::NewEnumerator1
    Loop Behavior=::NewEnumerator1
    Loop Duration Mode=::NewEnumerator1
    UseLoopDelay=false
   )
   &VC_FilteredSocketCount()
   &SpawnRate(SpawnRate=30, Use Spawn Probability=false) {enabled=false}
   &SpawnBurst_Instantaneous(Spawn Count=2, Use Spawn Probability=false, UseLoopCountLimit=false)
  ]
  particleSpawnScript [
   &VC_PaintedVariantsOverride() {enabled=false}
   &InitializeParticle(
    Color Mode=::NewEnumerator0
    Lifetime=4
    Mesh Scale=<3 2 4>
    Mesh Scale Mode=::NewEnumerator4
    Position Mode=::NewEnumerator0
    Write Material Random=true
   )
   &PartitionParticles(Number Of Partitions=$Emitter.PortCount)
   &SkeletalMeshLocation(
    Bone / Socket Index=$Particles.PartitionParticles.Partition
    Bone / Socket Sampling Mode=::NewEnumerator5
    Interpolate Mesh Sampling=false
    MaskSpawnGroup=false
    Mesh Sampling Type=::NewEnumerator6
    Sampled Position Offset=&AddVectorToPosition(Position=<0 0 0>, Vector=&AddVectorToPosition(Position=$User.DTCarOffset, Vector=<39.3897 0 0>))
    Skeletal Mesh=$System.SkeletalMesh
   )
   &AddVelocity(Velocity=&RandomRangeVector(Minimum=<-60 0 0>, Maximum=<-90 1 1>)) {enabled=false}
   &SpriteFacingAndAlignment(Sprite Alignment=<1 0 0>, Sprite Facing=<0 1 0>, WriteAlignment=true, WriteFacing=true) {enabled=false}
   &InitialMeshOrientation(Mesh Orientation Mode=::NewEnumerator7, Use Rotation Vector=false) {enabled=false}
   &DLR_ProductPreviews(GameModeID:NiagaraInt32=$System.GameModeID, bPreviewVehicleSpeed:NiagaraBool=$System.bPreviewVehicleSpeed)
  ]
  particleUpdateScript [
   &ParticleState(Kill Particles When Lifetime Has Elapsed=false, Loop Particles Lifetime=true)
   &InheritSourceMovement(
    Applied Position Scale=&Multiply_VectorByFloat(
     Vector=<1 1 1>
     Float=&FloatFromCurve(FloatCurve=DataInterface() {propertyValues {Curve {
         defaultValue=1
         keys [
          {arriveTangent=-1.7885, interpMode=RCIM_Cubic, leaveTangent=-1.7885, tangentMode=RCTM_User, time=0.05, value=1}
          {interpMode=RCIM_Cubic, tangentMode=RCTM_User, time=1, value=0}
         ]
        }}})
    )
    ApplyForce=false
    ApplyPosition=true
    ApplyPositionFalloff=false
   ) {enabled=false, module=InheritSourceMovement001}
   &ScaleColor(
    Scale Alpha=&FloatFromCurve(FloatCurve=DataInterface() {propertyValues {Curve {keys [{arriveTangent=0.3506, time=0, value=1}, <at 6 0>]}}})
    ScaleA=true
    ScaleRGB=false
   ) {enabled=false}
   &DynamicMaterialParameters(
    Index 0 Param 2=1
    Index 1 Param 1=1
    Index 1 Param 2=1
    Index 1 Param 3=0.992
    Index 1 Param 4=&FloatFromCurve(FloatCurve=DataInterface() {propertyValues {Curve {keys [<at 0 2>,<at 1 0>]}}})
    Index 2 Param 1=0.9488
    Index 2 Param 3=&FloatFromCurve(FloatCurve=DataInterface() {propertyValues {Curve {keys [<at 0 0>,<at 1 1>]}}})
    Index 2 Param 4=0.032
    Module.Index0Param0Write=false
    Module.Index0Param1Write=true
    Module.Index0Param2Write=false
    Module.Index0Param3Write=false
    Module.Index1Param0Write=true
    Module.Index1Param1Write=true
    Module.Index1Param2Write=false
    Module.Index1Param3Write=true
    Module.Index2Param0Write=true
    Module.Index2Param1Write=true
    Module.Index2Param2Write=true
    Module.Index2Param3Write=true
    Module.Index3Param0Write=true
    Module.Index3Param1Write=true
    Module.Index3Param2Write=true
    Module.Index3Param3Write=true
    Write Parameter Index 1=true
    Write Parameter Index 2=true
    Write Parameter Index 3=true
   )
   &JitterPosition(Jitter Amount=0.3) {enabled=false}
   &ScaleColor(
    Linear Color Curve=DataInterface() {propertyValues {
      AlphaCurve {keys [<at 0 0.5>]}
      BlueCurve {keys [<at 0.15 0>,<at 0.3 1>,<at 0.45 1>,<at 0.6 1>,<at 0.75 0>,<at 0.9 0>,<at 1 0.0072>]}
      GreenCurve {keys [<at 0.15 0>,<at 0.3 0>,<at 0.45 0>,<at 0.6 0.9661>,<at 0.75 1>,<at 0.9 0.8541>,<at 1 0.078>]}
      RedCurve {keys [<at 0.15 1>,<at 0.3 0.7185>,<at 0.45 0.0273>,<at 0.6 0.0289>,<at 0.75 0.0711>,<at 0.9 1>,<at 1 1>]}
     }}
    Scale Mode=2
   ) {module=ScaleColor001}
   &SolveForcesAndVelocity(Clamp Velocity=false, Limit Acceleration=false)
  ]
  rendererClasses [&NiagaraMeshRendererProperties, &NiagaraSpriteRendererProperties]
  renderers [
   {rendererClass=&NiagaraMeshRendererProperties, rendererIndex=0}
   {rendererClass=&NiagaraSpriteRendererProperties, rendererIndex=1}
  ]
 }
 NE_VC_Thruster002 {
  emitterUpdateScript [
   &EmitterState(
    Inactive Response=::NewEnumerator1
    Life Cycle Mode=::NewEnumerator1
    Loop Behavior=::NewEnumerator1
    Loop Duration Mode=::NewEnumerator1
    UseLoopDelay=false
   )
   &VC_FilteredSocketCount()
   &SpawnRate(SpawnRate=30, Use Spawn Probability=false) {enabled=false}
   &SpawnBurst_Instantaneous(Spawn Count=2, Use Spawn Probability=false, UseLoopCountLimit=false)
   /SetVariables(Emitter.HueShift:NiagaraFloat=&Divide_Float(A=&Add_Float(A=$Emitter.Age, B=0.5), B=3)) {isSetParameters=true, module=SetVariables_#4}
  ]
  particleSpawnScript [
   &VC_PaintedVariantsOverride() {enabled=false}
   &InitializeParticle(
    Color Mode=::NewEnumerator0
    Lifetime=4
    Mesh Scale=<4.5 2 4>
    Mesh Scale Mode=::NewEnumerator4
    Position Mode=::NewEnumerator0
    Write Material Random=true
   )
   &PartitionParticles(Number Of Partitions=$Emitter.PortCount)
   &SkeletalMeshLocation(
    Bone / Socket Index=$Particles.PartitionParticles.Partition
    Bone / Socket Sampling Mode=::NewEnumerator5
    Interpolate Mesh Sampling=false
    MaskSpawnGroup=false
    Mesh Sampling Type=::NewEnumerator6
    Sampled Position Offset=&AddVectorToPosition(Position=<0 0 0>, Vector=&AddVectorToPosition(Position=$User.DTCarOffset, Vector=<39.3897 0 0>))
    Skeletal Mesh=$System.SkeletalMesh
   )
   &AddVelocity(Velocity=&RandomRangeVector(Minimum=<-60 0 0>, Maximum=<-90 1 1>)) {enabled=false}
   &SpriteFacingAndAlignment(Sprite Alignment=<1 0 0>, Sprite Facing=<0 1 0>, WriteAlignment=true, WriteFacing=true) {enabled=false}
   &InitialMeshOrientation(Mesh Orientation Mode=::NewEnumerator7, Use Rotation Vector=false) {enabled=false}
   &DLR_ProductPreviews(GameModeID:NiagaraInt32=$System.GameModeID, bPreviewVehicleSpeed:NiagaraBool=$System.bPreviewVehicleSpeed)
  ]
  particleUpdateScript [
   &ParticleState(Kill Particles When Lifetime Has Elapsed=false, Loop Particles Lifetime=true)
   &InheritSourceMovement(
    Applied Position Scale=&Multiply_VectorByFloat(
     Vector=<1 1 1>
     Float=&FloatFromCurve(FloatCurve=DataInterface() {propertyValues {Curve {
         defaultValue=1
         keys [
          {arriveTangent=-1.7885, interpMode=RCIM_Cubic, leaveTangent=-1.7885, tangentMode=RCTM_User, time=0.05, value=1}
          {interpMode=RCIM_Cubic, tangentMode=RCTM_User, time=1, value=0}
         ]
        }}})
    )
    ApplyForce=false
    ApplyPosition=true
    ApplyPositionFalloff=false
   ) {enabled=false, module=InheritSourceMovement001}
   &ScaleColor(
    Scale Alpha=&FloatFromCurve(FloatCurve=DataInterface() {propertyValues {Curve {keys [{arriveTangent=0.3506, time=0, value=1}, <at 6 0>]}}})
    ScaleA=true
    ScaleRGB=false
   ) {enabled=false}
   &DynamicMaterialParameters(
    Index 0 Param 2=1
    Index 1 Param 1=1
    Index 1 Param 2=1
    Index 1 Param 3=0.992
    Index 1 Param 4=&FloatFromCurve(FloatCurve=DataInterface() {propertyValues {Curve {keys [<at 0 2>,<at 1 0>]}}})
    Index 2 Param 1=0.9488
    Index 2 Param 3=&FloatFromCurve(FloatCurve=DataInterface() {propertyValues {Curve {keys [<at 0 0>,<at 1 1>]}}})
    Index 2 Param 4=0.032
    Module.Index0Param0Write=false
    Module.Index0Param1Write=true
    Module.Index0Param2Write=false
    Module.Index0Param3Write=false
    Module.Index1Param0Write=true
    Module.Index1Param1Write=true
    Module.Index1Param2Write=false
    Module.Index1Param3Write=true
    Module.Index2Param0Write=true
    Module.Index2Param1Write=true
    Module.Index2Param2Write=true
    Module.Index2Param3Write=true
    Module.Index3Param0Write=true
    Module.Index3Param1Write=true
    Module.Index3Param2Write=true
    Module.Index3Param3Write=true
    Write Parameter Index 1=true
    Write Parameter Index 2=true
    Write Parameter Index 3=true
   )
   &JitterPosition(Jitter Amount=0.3) {enabled=false}
   &ScaleColor(
    Linear Color Curve=DataInterface() {propertyValues {
      AlphaCurve {keys [<at 0 1>,<at 1 1>]}
      BlueCurve {keys [<at 0 1>,<at 1 1>]}
      GreenCurve {keys [<at 0 1>,<at 1 1>]}
      RedCurve {keys [<at 0 1>,<at 1 1>]}
     }}
    Scale Mode=2
   ) {module=ScaleColor001}
   /SetVariables(Particles.HueShift:NiagaraFloat=$Emitter.Age) {isSetParameters=true, module=SetVariables_#5}
   &SolveForcesAndVelocity(Clamp Velocity=false, Limit Acceleration=false)
  ]
  rendererClasses [&NiagaraMeshRendererProperties, &NiagaraSpriteRendererProperties]
  renderers [
   {rendererClass=&NiagaraMeshRendererProperties, rendererIndex=0}
   {rendererClass=&NiagaraSpriteRendererProperties, rendererIndex=1}
  ]
 }
 Ribbons {
  bEnabled=false
  emitterUpdateScript [
   &VC_RateQualityHandler(Epic Spawn Rate=18, High Spawn Rate=18, Low Spawn Rate=18, Medium Spawn Rate=18, Use Custom Rate?=true) {module=VC_RateQualityHandler001}
   &EmitterState()
   &VC_FilteredSocketCount()
   &VC_RateQualityHandler(Default Rate:NiagaraFloat=&Multiply_Float(
     A=100
     B=&VC_Mask_Floats(Coro Value:NiagaraFloat=1, FrontEnd Value:NiagaraFloat=0.75, FN Value:NiagaraFloat=0.5)
    ))
   &VC_SpawnRibbon_v1(Number of Ribbons=$Emitter.PortCount, Spawn Time=10, SpawnRate=$Emitter.Spawn Rate)
  ]
  particleSpawnScript [
   &VC_PaintedVariantsOverride() {enabled=false}
   &VC_InitializeRibbon_v1(
    Color=$User.PaintedColor
    Custom Offset=true
    Enable Skeletal Position=true
    Flip Facing=true
    Lifetime=1.3
    Local Vector Offset=&AddVectorToPosition(Position=<100 0 0>, Vector=$User.DTCarOffset)
    Offset Amount=1
   )
   &AddVelocity(Velocity=&MakeVector(
     X=-1400
     Y=&Add_Float(
      A=&Sine(
       Normalized Angle=$Emitter.Age
       Period=0.1
       Scale=&FloatFromCurve(
        FloatCurve=DataInterface() {propertyValues {Curve {keys [{tangentMode=RCTM_User, time=0, value=0}, <at 0.6 1>, {tangentMode=RCTM_User, time=1, value=1}]}}}
        Scale Curve=50
       )
       Bias=0
      )
      B=&Sine(
       Normalized Angle=$Emitter.Age
       Period=0.3
       Scale=&FloatFromCurve(
        FloatCurve=DataInterface() {propertyValues {Curve {keys [{tangentMode=RCTM_User, time=0, value=0.1}, <at 0.6 1>, {tangentMode=RCTM_User, time=1, value=1}]}}}
        Scale Curve=50
       )
       Bias=0
      )
     )
     Z=&Sine(
      Normalized Angle=$Emitter.Age
      Period=0.1
      Scale=&FloatFromCurve(
       FloatCurve=DataInterface() {propertyValues {Curve {keys [{tangentMode=RCTM_User, time=0, value=0.1}, <at 0.6 1>, {tangentMode=RCTM_User, time=1, value=1}]}}}
       Scale Curve=55
      )
      Bias=0
     )
    )) {module=AddVelocity001}
   &VC_ExtraBindingColors(ExtraColors=2, R1 Color:LinearColor=$User.PaintedColor_2, R2 Color:LinearColor=$User.PaintedColor_3)
   &DLR_ProductPreviews(GameModeID:NiagaraInt32=$System.GameModeID, bPreviewVehicleSpeed:NiagaraBool=$System.bPreviewVehicleSpeed)
  ]
  particleUpdateScript [
   &ParticleState()
   &InheritSourceMovement(
    Applied Position Scale=&Multiply_VectorByFloat(
     Vector=<1 1 1>
     Float=&FloatFromCurve(FloatCurve=DataInterface() {propertyValues {Curve {keys [
          {interpMode=RCIM_Cubic, leaveTangent=-0.0051, tangentMode=RCTM_Break, time=0, value=1}
          {arriveTangent=-0.247, interpMode=RCIM_Cubic, tangentMode=RCTM_Break, time=1, value=0.9}
         ]}}})
    )
    ApplyForce=false
    ApplyPosition=true
    ApplyPositionFalloff=false
   ) {module=InheritSourceMovement001}
   &InheritSourceMovement(
    Applied Position Scale=&VectorFromFloat(Value=&FloatFromCurve(FloatCurve=DataInterface() {propertyValues {Curve {
         defaultValue=1
         keys [
          {arriveTangent=-1.7885, interpMode=RCIM_Cubic, leaveTangent=-1.7885, tangentMode=RCTM_User, time=0.05, value=1}
          {interpMode=RCIM_Cubic, tangentMode=RCTM_User, time=1, value=0}
         ]
        }}}))
    ApplyForce=false
    ApplyPosition=true
    ApplyPositionFalloff=false
   )
   &SolveForcesAndVelocity(Clamp Velocity=false, Limit Acceleration=false)
   &ScaleRibbonWidth(Ribbon Width Scale=&FloatFromCurve(
     FloatCurve=DataInterface() {propertyValues {Curve {keys [
         {interpMode=RCIM_Cubic, leaveTangent=1.6099, tangentMode=RCTM_Break, time=0, value=0.85}
         {interpMode=RCIM_Cubic, leaveTangent=-1.5386, tangentMode=RCTM_Break, time=0.08, value=1}
         {arriveTangent=-0.5728, interpMode=RCIM_Cubic, tangentMode=RCTM_Break, time=0.95, value=0}
        ]}}}
     Scale Curve=10
    ))
   &ScaleColor(
    Scale Alpha=&FloatFromCurve(
     FloatCurve=DataInterface() {propertyValues {Curve {keys [
         {arriveTangent=-7.9979, interpMode=RCIM_Cubic, leaveTangent=-7.9979, tangentMode=RCTM_User, time=0, value=0}
         {arriveTangent=12.4362, interpMode=RCIM_Cubic, leaveTangent=26.1544, tangentMode=RCTM_Break, time=0.1, value=-1.25}
         {arriveTangent=9.6937, interpMode=RCIM_Cubic, leaveTangent=4.0462, tangentMode=RCTM_Break, time=0.3, value=3}
         {arriveTangent=-2.3968, interpMode=RCIM_Cubic, leaveTangent=-20.7313, tangentMode=RCTM_Break, time=0.65, value=3}
         <at 1 0 RCIM_Cubic>
        ]}}}
     Scale Curve=2
    )
    Scale RGB=&VectorFromCurve(
     VectorCurve=DataInterface() {propertyValues {XCurve {keys [<at 0 1>,<at 1 1>]}, YCurve {keys [<at 0 1>,<at 1 1>]}, ZCurve {keys [<at 0 1>,<at 1 1>]}}}
     Scale Curve=<6 6 6>
    )
    ScaleA=true
    ScaleRGB=true
   ) {module=ScaleColor002}
   /SetVariables(
    Particles.R1 Color:LinearColor=&self:VC_Dynamic_ScaleColor_03(
     Color to Scale:LinearColor=$Particles.Initial.R1 Color
     RGB Scale:Vector3f=&VectorFromFloat(Value=&FloatFromCurve(FloatCurve=DataInterface() {propertyValues {Curve {keys [<at 0 1>,<at 1 1>]}}}, Scale Curve=6))
     Alpha Scale:NiagaraFloat=&FloatFromCurve(
      FloatCurve=DataInterface() {propertyValues {Curve {keys [
          {arriveTangent=-7.9979, interpMode=RCIM_Cubic, leaveTangent=-7.9979, tangentMode=RCTM_User, time=0, value=0}
          {arriveTangent=12.4362, interpMode=RCIM_Cubic, leaveTangent=26.1544, tangentMode=RCTM_Break, time=0.1, value=-1.25}
          {arriveTangent=9.6937, interpMode=RCIM_Cubic, leaveTangent=4.0462, tangentMode=RCTM_Break, time=0.3, value=3}
          {arriveTangent=-2.3968, interpMode=RCIM_Cubic, leaveTangent=-20.7313, tangentMode=RCTM_Break, time=0.65, value=3}
          <at 1 0 RCIM_Cubic>
         ]}}}
      Scale Curve=2
     )
    )
    Particles.R2 Color:LinearColor=&self:VC_Dynamic_ScaleColor_02(
     Color to Scale:LinearColor=$Particles.Initial.R2 Color
     RGB Scale:Vector3f=<6 6 6>
     Alpha Scale:NiagaraFloat=&FloatFromCurve(
      FloatCurve=DataInterface() {propertyValues {Curve {keys [
          {arriveTangent=-7.9979, interpMode=RCIM_Cubic, leaveTangent=-7.9979, tangentMode=RCTM_User, time=0, value=0}
          {arriveTangent=12.4362, interpMode=RCIM_Cubic, leaveTangent=26.1544, tangentMode=RCTM_Break, time=0.1, value=-1.25}
          {arriveTangent=9.6937, interpMode=RCIM_Cubic, leaveTangent=4.0462, tangentMode=RCTM_Break, time=0.3, value=3}
          {arriveTangent=-2.3968, interpMode=RCIM_Cubic, leaveTangent=-20.7313, tangentMode=RCTM_Break, time=0.65, value=3}
          <at 1 0 RCIM_Cubic>
         ]}}}
      Scale Curve=0.4
     )
    )
    Particles.RibbonWidth2:NiagaraFloat=&FloatFromCurve(FloatCurve=DataInterface() {propertyValues {Curve {keys [<at 0 1>,<at 1 0>]}}}, Scale Curve=20)
    Particles.RibbonWidth3:NiagaraFloat=&FloatFromCurve(
     FloatCurve=DataInterface() {propertyValues {Curve {keys [
         {interpMode=RCIM_Cubic, leaveTangent=-1.1418, tangentMode=RCTM_Break, time=0, value=1}
         {arriveTangent=-1.0164, tangentMode=RCTM_User, time=0.5, value=0.4}
         <at 0.85 0>
        ]}}}
     Scale Curve=122
    )
   ) {isSetParameters=true}
   &VC_UpdateRibbon_v1(Enable Skeletal Position=true)
  ]
  rendererClasses [&NiagaraRibbonRendererProperties]
  renderers [
   {rendererClass=&NiagaraRibbonRendererProperties}
   {rendererClass=&NiagaraRibbonRendererProperties, rendererIndex=1}
   {rendererClass=&NiagaraRibbonRendererProperties, rendererIndex=2}
  ]
 }
}
summary {
 systemName=NS_Irid
 emitters [
  {emitterName=NE_VC_Boost_Ribbon_Parent, rendererClasses [&NiagaraRibbonRendererProperties]}
  {bEnabled=false, emitterName=NE_VC_Boost_Ribbon_Parent001, rendererClasses [&NiagaraRibbonRendererProperties]}
  {
   bEnabled=false
   emitterName=NE_VC_Thruster001
   rendererClasses [&NiagaraMeshRendererProperties, &NiagaraSpriteRendererProperties]
  }
  {emitterName=NE_VC_Thruster, rendererClasses [&NiagaraMeshRendererProperties, &NiagaraSpriteRendererProperties]}
  {
   emitterName=NE_VC_Thruster002
   rendererClasses [&NiagaraMeshRendererProperties, &NiagaraSpriteRendererProperties]
  }
  {bEnabled=false, emitterName=EnergyStrike_Optimized, rendererClasses [&NiagaraRibbonRendererProperties]}
  {bEnabled=false, emitterName=Ribbons, rendererClasses [&NiagaraRibbonRendererProperties]}
 ]
 userVariables [
  User.bIsFrontendPreview {type=NiagaraBool}
  User.bIsPainted {type=NiagaraBool}
  User.blsIsCoroColor {type=NiagaraBool}
  User.bPreviewVehicleSpeed {type=NiagaraBool}
  User.ColorVariantID {type=NiagaraInt32}
  User.DTCarOffset {type=NiagaraPosition, description="", defaultValue {}}
  User.DTCarOffset001 {type=Vector3f, description="", defaultValue {struct=Vector3f, value=<0 0 0>}}
  User.GameModeID {type=NiagaraInt32, defaultValue=3}
  User.HueShift_M1 {type=NiagaraFloat, defaultValue=0.05, description=""}
  User.HueShift_M2 {type=NiagaraFloat, defaultValue=0, description=""}
  User.HueShift_M3 {type=NiagaraFloat, defaultValue=0.05, description=""}
  User.HueShift_M4 {type=NiagaraFloat, defaultValue=0.2, description=""}
  User.PaintedColor {type=LinearColor, description="", defaultValue {struct=LinearColor, value=<rgba 1 1 1 1>}}
  User.PaintedColor_2 {type=LinearColor, description="", defaultValue {struct=LinearColor, value=<rgba 0.102 0.502 1 1.5>}}
  User.PaintedColor_3 {type=LinearColor, description="", defaultValue {struct=LinearColor, value=<rgba 0.7728 0.0827 1 2>}}
  User.PaintedScalar {type=NiagaraFloat, defaultValue=1, description=""}
  User.PortCount {type=NiagaraInt32, defaultValue=2}
  User.Power_AlphaPaintedR1 {type=NiagaraFloat, defaultValue=1, description=""}
  User.Power_AlphaPaintedR2 {type=NiagaraFloat, defaultValue=1, description=""}
  User.Power_AlphaPaintedR3 {type=NiagaraFloat, defaultValue=1, description=""}
  User.Power_AlphaPaintedR4 {type=NiagaraFloat, defaultValue=1, description=""}
  User.Power_EmissivePaintedR1 {type=NiagaraFloat, defaultValue=1, description=""}
  User.Power_EmissivePaintedR2 {type=NiagaraFloat, defaultValue=1, description=""}
  User.Power_EmissivePaintedR3 {type=NiagaraFloat, defaultValue=1, description=""}
  User.Power_EmissivePaintedR4 {type=NiagaraFloat, defaultValue=1, description=""}
 ]
}
